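# Physics asset for the Punchy two-handed melee pickaxe, defining collision bodies for its handle and glove bones.
# A capsule covers the handle, a large offset box covers dyn_glove, and a small box sits on dyn_glove_mstr.
# The single constraint joins dyn_glove to dyn_glove_mstr with swing and twist locked, so the glove cannot rotate relative to that bone.
# The handle body has no constraint to either glove body.
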PhysicsAsset /BRCosmetics/Weapons/FORT_Melee/Pickaxe_Punchy/Meshes/Punchy_Pickaxe_PhysicsAsset
bodies {
 Melee_TwoHanded_Handle [{shapeType=Capsule, length=92.837, radius=7.9529, center=<0.638 -0.887 28.4253>, rotation=<rot 1.85 -90 0.2974>}]
 dyn_glove [{shapeType=Box, extentX=20.9972, extentY=37.4287, extentZ=24.3406, center=<0.4142 27.2708 -1.453>}]
 dyn_glove_mstr [{shapeType=Box, extentX=7.5, extentY=7.5, extentZ=7.5}]
}
constraints [{
  bone1Name=dyn_glove
  bone2Name=dyn_glove_mstr
  swing1LimitDegrees=45
  swing1Motion=Locked
  swing2LimitDegrees=45
  swing2Motion=Locked
  twistLimitDegrees=45
  twistMotion=Locked
 }]
